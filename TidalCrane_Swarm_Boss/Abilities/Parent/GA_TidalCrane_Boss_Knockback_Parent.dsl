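; Parent gameplay ability for the TidalCrane swarm boss that damages and knocks back targets, with empty activate and end events.
; TryKnockbackWithGameplayEffect runs target selection, then applies the tagged effect container and loops over target actors.
; Pawns are launched away from the avatar using scalable XY and Z velocities.
; Other actors are cast to vehicles, whose similarly built vector feeds a BroadcastAbilityImpactatLocation call instead of a launch.
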(event Ability|EventActivateAbility)

(event Ability|EventOnEndAbility
  (params (bWasCancelled :type Boolean)))

(event Custom|TryKnockbackWithGameplayEffect
  ; Call the GameplayEffectContainer
  (bind (_asynctaskproxy _targetdata _applicationtag) (Ability|Tasks|DoTargetSelection "None" "(TagName=\"Targeting.ApplicationTag.0\")")
    (:Targeted
      (CallFunction|ApplyDamageandKnockback _targetdata _applicationtag))))

(event Custom|ApplyDamageAndKnockback
  (params (TargetData :type Gameplay%20Ability%20Target%20Data%20Handle%20Structure) (ApplicationTag :type Gameplay%20Tag%20Structure))
  ; Loop Actors and knock them back on their type
  (Ability|ApplyGameplayEffectContainer TargetData ApplicationTag)
  (for _array_element (Ability|TargetData|GetAllActorsfromTargetData TargetData)
    (bind _asfort_pawn (Utilities|Casting|CastToFortPawn _array_element)
      (:then
        (Default|KnockbackPawns _asfort_pawn))
      (:CastFailed
        (bind _asfort_athena_vehicle (Utilities|Casting|CastToFortAthenaVehicle _array_element))
        (Default|KnockbackVehicles _asfort_athena_vehicle)))))

(fn KnockbackPawns
  (params (Pawn :type Fort%20Pawn%20Object%20Reference))
  (bind (_returnvalue_x _returnvalue_y) (Math|Vector|BreakVector (Math|Vector|Normalize (- (Transformation|GetActorLocation Pawn) (Transformation|GetActorLocation (Ability|GetAvatarActorfromActorInfo))))))
  (Character|LaunchCharacter Pawn (Math|Vector|MakeVector (* _returnvalue_x (ScalableFloat|GetValueatLevel (Variables|Default|GetKnockbackPawn_XYVelocity))) (* _returnvalue_y (ScalableFloat|GetValueatLevel (Variables|Default|GetKnockbackPawn_XYVelocity))) (ScalableFloat|GetValueatLevel (Variables|Default|GetKnockbackPawn_ZVelocity))) true true))

(fn KnockbackVehicles
  (params (FortAthenaVehicle :type Fort%20Athena%20Vehicle%20Object%20Reference))
  (bind (_returnvalue_x _returnvalue_y) (Math|Vector|BreakVector (Math|Vector|Normalize (- (Transformation|GetActorLocation FortAthenaVehicle) (Transformation|GetActorLocation (Ability|GetAvatarActorfromActorInfo))))))
  (Audio|BroadcastAbilityImpactatLocation 0 0 (Math|Vector|MakeVector (* _returnvalue_x (ScalableFloat|GetValueatLevel (Variables|Default|GetKnockbackVehicle_XYVelocity))) (* _returnvalue_y (ScalableFloat|GetValueatLevel (Variables|Default|GetKnockbackVehicle_XYVelocity))) (ScalableFloat|GetValueatLevel (Variables|Default|GetKnockbackVehicle_ZVelocity))) 777.0 "None" false))
